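; Artemis Daisy small platform cap prop, a building actor Blueprint variant that, by its name, does not snap.
; It overrides footstep surface queries, returning the default footstep surface data and the default surface type.
; The construction script is empty, so the asset carries no other logic of its own.
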(fn GetFootstepSurfaceData
  (return (Variables|Default|GetFootstepSurfaceData)))

(fn GetSurfaceType
  (return "SurfaceType_Default"))

(fn ConstructionScript)
