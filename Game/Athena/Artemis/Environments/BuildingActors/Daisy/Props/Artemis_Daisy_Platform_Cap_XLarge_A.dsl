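; Building actor Blueprint for an extra-large platform cap prop (variant A) in the Artemis Daisy environment set.
; Getter functions pass through variables for footstep surface data, marker display data and marker position offset.
; The surface type getter returns the default surface, and the construction script is empty, so the graphs add no logic of their own.
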(fn GetFootstepSurfaceData
  (return (Variables|Default|GetFootstepSurfaceData)))

(fn GetMarkerDisplayData
  (return (Variables|Marking|GetMarkerDisplay)))

(fn GetMarkerPositionOffset
  (return (Variables|Marking|GetMarkerPositionOffset)))

(fn GetSurfaceType
  (return "SurfaceType_Default"))

(fn ConstructionScript)
