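; Blueprint for a Reality Bloom environment prop (Close A variant) that feeds world-space pivot data to its mesh materials.
; On BeginPlay it calls the parent, then writes the PivotMin world location to the WSCanopyPivot vector parameter and the PivotMax location to MaxHeight.
; These values are set once at runtime on the static mesh component's materials, as the construction script is empty.
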(event EventBeginPlay
  (bind _staticmeshcomponent (Variables|Debug|GetStaticMeshComponent))
  (|CallParentFunction)
  (Rendering|Material|SetVectorParameterValueonMaterials _staticmeshcomponent "WSCanopyPivot" (Transformation|GetWorldLocation (Variables|Default|GetPivotMin)))
  (Rendering|Material|SetVectorParameterValueonMaterials _staticmeshcomponent "MaxHeight" (Transformation|GetWorldLocation (Variables|Default|GetPivotMax))))

(fn ConstructionScript)
