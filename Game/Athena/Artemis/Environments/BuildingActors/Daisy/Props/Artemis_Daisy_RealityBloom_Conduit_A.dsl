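; Reality Bloom Conduit prop Blueprint for the Daisy environment that drives material pivots and a local sound mix.
; On BeginPlay it writes PivotMin and PivotMax world locations to the static mesh's WSCanopyPivot and MaxHeight material parameters.
; PP_Bounds overlaps push the SoundMix modifier when a player pawn enters and pop it on exit; a non-pawn entering also pops it.
; SpawnDialog and the construction script are empty.
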(event EventBeginPlay
  (bind _staticmeshcomponent (Variables|Debug|GetStaticMeshComponent))
  (|CallParentFunction)
  (Rendering|Material|SetVectorParameterValueonMaterials _staticmeshcomponent "WSCanopyPivot" (Transformation|GetWorldLocation (Variables|Default|GetPivotMin)))
  (Rendering|Material|SetVectorParameterValueonMaterials _staticmeshcomponent "MaxHeight" (Transformation|GetWorldLocation (Variables|Default|GetPivotMax))))

(event OnComponentBeginOverlap(PP_Bounds)
  (params (OverlappedComponent :type Primitive%20Component%20Object%20Reference) (OtherActor :type Actor%20Object%20Reference) (OtherComp :type Primitive%20Component%20Object%20Reference) (OtherBodyIndex :type Integer) (bFromSweep :type Boolean) (SweepResult :type Hit%20Result%20Structure%20%28by%20ref%29))
  (bind _asfort_player_pawn_athena (Utilities|Casting|CastToFortPlayerPawnAthena OtherActor)
    (:then
      (if (== _asfort_player_pawn_athena OtherActor)
        (Audio|PushSoundMixModifier (Variables|Default|GetSoundMix))))
    (:CastFailed
      (Audio|PopSoundMixModifier (Variables|Default|GetSoundMix)))))

(event OnComponentEndOverlap(PP_Bounds)
  (params (OverlappedComponent :type Primitive%20Component%20Object%20Reference) (OtherActor :type Actor%20Object%20Reference) (OtherComp :type Primitive%20Component%20Object%20Reference) (OtherBodyIndex :type Integer))
  (bind _asfort_player_pawn_athena (Utilities|Casting|CastToFortPlayerPawnAthena OtherActor))
  (if (== _asfort_player_pawn_athena OtherActor)
    (Audio|PopSoundMixModifier (Variables|Default|GetSoundMix))))

(fn SpawnDialog)

(fn ConstructionScript)
