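; Blueprint building-actor prop for an open Reality Bloom, which feeds world-space bounds to the materials on its giant tree canopy mesh.
; On BeginPlay it calls the parent, then runs UpdateCanopyBounds.
; That function writes the world locations of the PivotRef and MaxPivotRef components into the WSCanopyPivot and MaxHeight vector parameters.
; The construction script is empty, so this graph sets the values only at play start.
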(event EventBeginPlay
  (|CallParentFunction)
  (CallFunction|UpdateCanopyBounds))

(fn UpdateCanopyBounds
  (bind _sm_gianttree_canopy_lg_02 (Variables|Default|GetSM_GiantTree_Canopy_LG_02))
  (Rendering|Material|SetVectorParameterValueonMaterials _sm_gianttree_canopy_lg_02 "WSCanopyPivot" (Transformation|GetWorldLocation (Variables|Default|GetPivotRef)))
  (Rendering|Material|SetVectorParameterValueonMaterials _sm_gianttree_canopy_lg_02 "MaxHeight" (Transformation|GetWorldLocation (Variables|Default|GetMaxPivotRef))))

(fn ConstructionScript)
